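# Unlit, two-sided translucent trail material for the Stereo Female glider's Niagara ribbons and sprites.
# A panning plasma lightning texture, offset per particle, is distorted by noise that jumps in time steps through a Ceil node.
# The same noise and U-based ramps erode and fade the trail, which particle color tints.
# Multiplied sine waves with particle-random phase and adjustable contrast add alpha flicker.
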Material /BRCosmetics/Gadgets/Assets/VinderTech_GliderChute/Glider_Stereo_Female/FX/M_Glider_Stereo_Trail
graph {
 properties {
  bUsedWithNiagaraRibbons=true
  bUsedWithNiagaraSprites=true
  blendMode=BLEND_Translucent
  pixelDepthOffsetMode=PDOM_Legacy
  shadingModel=MSM_Unlit
  twoSided=true
  usedShadingModels=Unlit
 }
 nodes [
  n1 = ParticleColor()
  n2 = MaterialFunctionCall(XY=TextureCoordinate() {coordinateIndex=1}) {materialFunction=&Swizzle}
  n3 = Add(A=Time(), B=DynamicParameter() {paramNames [RandomOffset,Param2,Param3,Param4]}.RandomOffset)
  n4 = AppendVector(A=Constant(), B=Constant() {r=-1.5})
  n5 = Ceil(None=Multiply(A=Time(), B=ScalarParameter() {defaultValue=16, parameterName=UVDistortSpeed}) {constB=16})
  n6 = TextureSample(UVs=Panner(Coordinate=TextureCoordinate() {coordinateIndex=1, uTiling=0.75, vTiling=0.75}, Time=Divide(A=@n5) {constB=4}) {speedX=-1, speedY=0.4}) {samplerType=SAMPLERTYPE_Masks, texture=/Game/Global/Textures/Default/Noise/T_WrapEdgeWear_Noise}
  n7 = AppendVector(A=@n6.R, B=@n6.R)
  n8 = TextureSample(UVs=Add(A=Panner(Coordinate=@n2.YX, Time=@n3, Speed=@n4), B=Multiply(A=@n7) {constB=0.015})) {texture=/Game/Effects/Fort_Effects/Textures/Tile/T_Lightning_Plasma_Soft}
  n9 = Multiply(A=@n8.B, B=ScalarParameter() {defaultValue=1})
  n10 = ComponentMask(None=TextureCoordinate()) {r=true}
  n11 = Saturate(None=Subtract(A=@n6.R, B=Subtract(A=@n10) {constB=0.5}))
  n12 = Saturate(None=Subtract(A=Multiply(A=@n10) {constB=5}) {constB=0.1})
  n13 = Power(Base=Multiply(A=Add(A=@n8.R, B=@n9), B=Multiply(A=@n11, B=@n12))) {constExponent=1.5}
  MaterialFunctionCall(Emissive=Multiply(A=@n13, B=@n1.RGB)) {materialFunction=&MF_FX_EmissiveAdjustments}
  n14 = LinearInterpolate(A=ParticleRandom(), B=Constant())
  n15 = Multiply(A=Add(A=Time(), B=Add(A=@n14)))
  n16 = Sine(None=@n15) {period=0.11}
  n17 = Multiply(A=Sine(None=@n15) {period=0.09}, B=Sine(None=@n15) {period=0.075})
  n18 = MaterialFunctionCall(In=ConstantBiasScale(None=Multiply(A=@n16, B=@n17)), Contrast=ScalarParameter() {defaultValue=3, parameterName=FlickerContrast}) {materialFunction=&CheapContrast}
  Multiply(A=@n1.A, B=Multiply(A=Saturate(None=@n18.Result), B=Clamp(None=@n13) {clampMode=CMODE_ClampMax}))
 ]
}
